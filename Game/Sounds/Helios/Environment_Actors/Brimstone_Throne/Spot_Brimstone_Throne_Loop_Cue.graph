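# Looping ambient spot sound for the Brimstone Throne environment prop, mixing two reused loops.
# A Brimstone pillar low-end loop and a pitched-down glowing rock loop swell in opposite directions on 20-second envelopes, crossbreathing in the mix.
# The rock layer also runs through a 60-second looping pitch envelope.
# Overrides set cone attenuation with listener focus, a distant reverb send and a two-voice concurrency limit.
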SoundCue /Game/Sounds/Helios/Environment_Actors/Brimstone_Throne/Spot_Brimstone_Throne_Loop_Cue
root=Mixer_2
nodes [
 {
  id=Enveloper_0
  children [Modulator_4]
  properties {
   bLoop=true
   bLoopIndefinitely=true
   loopEnd=60
   pitchCurve {editorCurveData {keys [
      {interpMode=RCIM_Cubic, time=0}
      {arriveTangent=0.0091, interpMode=RCIM_Cubic, leaveTangent=0.0091, tangentMode=RCTM_User, time=30, value=0.8}
      {time=60}
     ]}}
   volumeCurve {editorCurveData {keys [<at 0 1>]}}
  }
 }
 {
  id=Enveloper_4
  children [Enveloper_0]
  properties {
   bLoop=true
   bLoopIndefinitely=true
   loopEnd=20
   pitchCurve {editorCurveData {keys [{time=0}]}}
   volumeCurve {editorCurveData {keys [
      {arriveTangent=-0.01, interpMode=RCIM_Cubic, time=0, value=1}
      <at 10 1.2 RCIM_Cubic>
      {interpMode=RCIM_Cubic, leaveTangent=0.01, time=20, value=1}
     ]}}
  }
 }
 {
  id=Enveloper_7
  children [WavePlayer_46]
  properties {
   bLoop=true
   bLoopIndefinitely=true
   loopEnd=20
   pitchCurve {editorCurveData {keys [{time=0}]}}
   volumeCurve {editorCurveData {keys [
      {arriveTangent=-0.01, interpMode=RCIM_Cubic, time=0, value=1}
      <at 10 0.8 RCIM_Cubic>
      {interpMode=RCIM_Cubic, leaveTangent=0.01, time=20, value=1}
     ]}}
  }
 }
 {id=Mixer_2, children [Enveloper_7,Enveloper_4], properties {inputVolume [0.8,2]}}
 {id=Modulator_4, children [WavePlayer_1], properties {pitchMax=0.8, pitchMin=0.8}}
 {
  id=WavePlayer_1
  properties {bLooping=true, soundWaveAssetPtr=/Game/Sounds/Asteria/Environment_Actors/Fortnitanium/Spot_Fornitanium_Glowing_Rock_Loop}
 }
 {
  id=WavePlayer_46
  properties {bLooping=true, soundWaveAssetPtr=/Game/Sounds/Helios/Environment_Actors/BrimestoneDecoPillar/Spot_Brimstone_Pillar_Proximity_LowEnd_Loop}
 }
]
properties {
 bOverrideAttenuation=true
 bOverrideConcurrency=true
 maxDistance=1000
 soundClassObject=&Fort_Ambient_Props_Athena
 volumeMultiplier=0.251
 attenuationOverrides {
  attenuationShape=Cone
  bEnableFocusInterpolation=true
  bEnableListenerFocus=true
  bEnableSubmixSends=true
  falloffDistance=450
  focusVolumeAttenuation=1.1
  nonFocusVolumeAttenuation=0.5
  nonSpatializedRadiusStart=500
  stereoSpread=100
  attenuationShapeExtents=<550 80>
  submixSendSettings [{
    bEnableHPFCutoff=false
    bEnableLPFCutoff=false
    disableManualSendClamp=false
    hPFCutoff=20
    lPFCutoff=20000
    maxSendDistance=7500
    maxSendLevel=0.25
    minSendDistance=0
    minSendLevel=0
    sendLevel=0.2
    sendLevelControlMethod=Linear
    soundSubmix=&Fortnite_Reverb_Distant
   }]
 }
 concurrencyOverrides {maxCount=2, voiceStealReleaseTime=2}
}
